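# Niagara effect for the JethroJelly hover-jet jetpack's failed ascent: sputtering exhaust flames, glow, sparks and smoke.
# Partitioned particles follow User.Skeletal Mesh sockets, with mesh flames from burst and two rate emitters, plus flare sprites and lights.
# User.FuelFlicker drives color and waveform-based flicker, while User.uPhase scales flame size.
# Sparks and smoke burst, then spawn faster over the emitter loop; the Head ribbon emitter is disabled.
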NiagaraSystem /JethroWeaponGameplay/Assets/FX/Jetpack/Fail/NS_JethroJelly_HoverJets_AscendFail
dependencies {
 usedDataInterfaces [
  &NiagaraDataInterfaceColorCurve
  &NiagaraDataInterfaceVectorCurve
  &NiagaraDataInterfaceCurve
  &NiagaraDataInterfaceArrayFloat
 ]
 usedDynamicInputs [&SetBoolByFloatComparison]
 usedModules [&SystemState, &Timeline]
}
emitters {
 Airglow_Light_Flare {
  emitterUpdateScript [
   &EmitterState()
   &SpawnRate(SpawnRate=48, Use Spawn Probability=false)
  ]
  particleSpawnScript [
   &PartitionParticles()
   &InitializeParticle(
    Color=<rgba 1 1 1 0.25>
    Lifetime=0.17
    Sprite Size Mode=::NewEnumerator3
    Uniform Sprite Size=150
    UsePositionOffset=true
    Write Material Random=true
   )
   &CameraOffset()
  ]
  particleUpdateScript [
   &ParticleState()
   &ScaleSpriteSize(Uniform Curve Sprite Scale=DataInterface() {propertyValues {Curve {keys [{interpMode=RCIM_Cubic, leaveTangent=1.5808, tangentMode=RCTM_Break, time=0, value=0.5}, {tangentMode=RCTM_User, time=1, value=1}]}}})
   &FN_Sprite_Attributes(
    Color=&MultiplyLinearColor(
     Linear Color=&ColorFromCurve(
      ColorCurve=DataInterface() {propertyValues {
        AlphaCurve {keys [<at 1 0.3 RCIM_Cubic>]}
        BlueCurve {keys [<at 0 0>,<at 0.15 0>,<at 1 0>]}
        GreenCurve {keys [<at 0 0.125>,<at 0.15 0.6667>,<at 1 0.25>]}
        RedCurve {keys [<at 0 1>,<at 0.15 4>,<at 1 1>]}
       }}
      CurveIndex=$User.FuelFlicker
     )
     Scale Factor=<1 1 1 1>
    )
    Create Camera Offset Attribute=true
    Create Color Attribute=true
    "Create Normalized Age "=true
    Create SpriteSize Attribute=true
    SpriteSize=&Multiply_Vector2D_ByFloat(
     Float=&FloatFromCurve(FloatCurve=DataInterface() {propertyValues {Curve {keys [
          {interpMode=RCIM_Cubic, leaveTangent=13.149, tangentMode=RCTM_Break, time=0, value=0}
          {interpMode=RCIM_Cubic, tangentMode=RCTM_Break, time=0.2, value=1.1}
         ]}}})
     Vector2D=<50 50>
    )
    Write to Module Namespace=true
   )
   &FN_Light_Attributes(
    Color=&ColorFromCurve(
     ColorCurve=DataInterface() {propertyValues {
       AlphaCurve {keys [<at 1 1 RCIM_Cubic>]}
       BlueCurve {keys [<at 0 0>,<at 0.15 0>,<at 1 0>]}
       GreenCurve {keys [<at 0 0.125>,<at 0.15 0.6667>,<at 1 0.25>]}
       RedCurve {keys [<at 0 1>,<at 0.15 4>,<at 1 1>]}
      }}
     CurveIndex=$User.FuelFlicker
    )
    Create Color Attribute=true
    Create Exponent Attribute=true
    Create Radius Attribute=true
    Enable Particle Light Enable=true
    Light Enabled=true
    Radius=&FloatFromCurve(
     FloatCurve=DataInterface() {propertyValues {Curve {keys [
         {arriveTangent=9.0725, interpMode=RCIM_Cubic, leaveTangent=9.0725, tangentMode=RCTM_User, time=0, value=0}
         {arriveTangent=-2.5249, interpMode=RCIM_Cubic, leaveTangent=-2.5249, tangentMode=RCTM_User, time=0.1408, value=0.9847}
         {arriveTangent=-0.0405, interpMode=RCIM_Cubic, leaveTangent=-0.0405, tangentMode=RCTM_User, time=1, value=0}
        ]}}}
     Scale Curve=500
    )
   )
   &FN_Sprite_Attributes(
    Color=&MultiplyLinearColor(
     Linear Color=&ColorFromCurve(
      ColorCurve=DataInterface() {propertyValues {
        AlphaCurve {keys [<at 1 0.3 RCIM_Cubic>]}
        BlueCurve {keys [<at 0 0>,<at 0.15 0>,<at 1 0>]}
        GreenCurve {keys [<at 0 0.125>,<at 0.15 0.6667>,<at 1 0.25>]}
        RedCurve {keys [<at 0 1>,<at 0.15 4>,<at 1 1>]}
       }}
      CurveIndex=$User.FuelFlicker
     )
     Scale Factor=<1 1 1 1>
    )
    Create Camera Offset Attribute=true
    Create Color Attribute=true
    Create SpriteSize Attribute=true
    SpriteSize=&Multiply_Vector2D_ByFloat(
     Float=&RandomRangeFloat(Minimum=0.9, Maximum=1, Evaluation Type=::NewEnumerator1)
     Vector2D=&Lerp_Vector2(A=<31 7>, B=<125 10>, Alpha=$User.FuelFlicker)
    )
    Write to Module Namespace=true
   ) {module=FN_Sprite_Attributes001}
   &ScaleColor(
    Curve Index=$User.FuelFlicker
    Linear Color Curve=DataInterface() {propertyValues {
      AlphaCurve {keys [<at 1 0.3 RCIM_Cubic>]}
      BlueCurve {keys [<at 0.15 0>,<at 1 1 RCIM_Cubic>]}
      GreenCurve {keys [<at 0.15 0.4>,<at 1 0.4148 RCIM_Cubic>]}
      RedCurve {keys [<at 0.15 4>,<at 1 0 RCIM_Cubic>]}
     }}
    Scale Mode=::NewEnumerator2
   )
   &SkeletalMeshLocation(
    Bone / Socket Index=$Particles.PartitionParticles.Partition
    MaskSpawnGroup=false
    Mesh Sampling Type=::NewEnumerator5
    Sampled Position Offset=<7 0 0>
    Skeletal Mesh=$User.Skeletal Mesh
    Socket Sampling Mode=::NewEnumerator5
   )
   &DynamicMaterialParameters(
    Index 0 Param 1=0.5
    Index 0 Param 2=1.5
    Index 0 Param 3=1
    Index 0 Param 4=1
    Module.Index0Param0Write=true
    Module.Index0Param1Write=true
    Module.Index0Param2Write=true
    Module.Index0Param3Write=true
   )
  ]
  rendererClasses [&NiagaraSpriteRendererProperties, &NiagaraLightRendererProperties]
  renderers [
   {rendererClass=&NiagaraSpriteRendererProperties}
   {rendererClass=&NiagaraSpriteRendererProperties, rendererIndex=1}
   {rendererClass=&NiagaraLightRendererProperties, rendererIndex=2}
   {rendererClass=&NiagaraSpriteRendererProperties, rendererIndex=3}
  ]
 }
 Burst {
  emitterUpdateScript [
   &EmitterState()
   &SpawnBurst_Instantaneous(Spawn Count=2, Use Spawn Probability=false, UseLoopCountLimit=false)
  ]
  particleSpawnScript [
   &PartitionParticles()
   &InitializeParticle(
    Color Mode=::NewEnumerator0
    Lifetime Max=0.25
    Lifetime Min=0.15
    Lifetime Mode=::NewEnumerator1
    Position Mode=::NewEnumerator0
    Write Material Random=true
   )
   /SetVariables(
    Particles.Color_Blaster:LinearColor=<rgba 20 1.6667 0 0.1>
    Particles.Orientation_Blaster:Quat4f=&OrientMeshToVector(Look At Direction=<0 -1 0>, Reference Axis Vector=<1 0 0>)
    Particles.Size_Blaster:Vector3f=<2.5 1 1>
   ) {isSetParameters=true, module=SetVariables_#1}
  ]
  particleUpdateScript [
   /SetVariables(
    Particles.Color_Blaster:LinearColor=&MultiplyLinearColor(
     Linear Color=&ColorFromCurve(ColorCurve=DataInterface() {propertyValues {
        AlphaCurve {keys [<at 0 1 RCIM_Cubic>,<at 0.9898 0>]}
        BlueCurve {keys [<at 0 0 RCIM_Cubic>,<at 1 0>]}
        GreenCurve {keys [<at 0 1.2915 RCIM_Cubic>,<at 1 0.0667>]}
        RedCurve {keys [<at 0 7 RCIM_Cubic>,<at 1 4>]}
       }})
     Scale Factor=&MakeFloat4FromVectorAndFloat(
      "Vector (RGB)"=&VectorFromCurve(VectorCurve=DataInterface() {propertyValues {
         XCurve {keys [<at 0 2 RCIM_Cubic>,<at 1 1 RCIM_Cubic>]}
         YCurve {keys [<at 0 2 RCIM_Cubic>,<at 1 1 RCIM_Cubic>]}
         ZCurve {keys [<at 0 2 RCIM_Cubic>,<at 1 1 RCIM_Cubic>]}
        }})
      "Float (Alpha)"=&Lerp_Float(
       A=&Waveform("Waveform [1]"=::NewEnumerator8, "[1] Frequency"=1, "[1] Amplitude Min/Max"=<0 1>, "[1] Scale"=2)
       B=&FloatFromCurve(
        FloatCurve=DataInterface() {propertyValues {Curve {keys [<at 0 0 RCIM_Cubic>, {interpMode=RCIM_Cubic, tangentMode=RCTM_Break, time=0.5, value=1}, <at 1.0011 0 RCIM_Cubic>]}}}
        Scale Curve=&Lerp_Float(
         A=$User.FuelFlicker
         B=1
         Alpha=&Waveform("Waveform [1]"=::NewEnumerator8, "[1] Phase Scale"=50, "[1] Frequency"=5, "[1] Amplitude Min/Max"=<0 1>)
        )
       )
       Alpha=$User.FuelFlicker
      )
     )
    )
    Particles.DynamicParameter_Blaster:Vector4f=&MakeVector4(
     X=&FloatFromCurve(FloatCurve=DataInterface() {propertyValues {Curve {keys [<at 0 1>,<at 1 0>]}}})
     Y=5
     Z=20
     W=1
    )
    Particles.Size_Blaster:Vector3f=&Multiply_VectorByFloat(
     Vector=&Multiply_Vector(
      A=&VectorFromCurve(
       VectorCurve=DataInterface() {propertyValues {
         XCurve {keys [
           <at 0 0.75 RCIM_Cubic>
           <at 0.4908 3.9579>
           {arriveTangent=4.355, interpMode=RCIM_Cubic, leaveTangent=4.355, tangentMode=RCTM_User, time=0.5, value=4}
           <at 1 0.5 RCIM_Cubic>
          ]}
         YCurve {keys [
           <at 0 1 RCIM_Cubic>
           <at 0.4908 0.9933>
           {arriveTangent=0.75, interpMode=RCIM_Cubic, leaveTangent=0.75, time=0.5, value=1}
           <at 1 1.75 RCIM_Cubic>
          ]}
         ZCurve {keys [<at -0.0014 1 RCIM_Cubic>, {arriveTangent=0.7489, interpMode=RCIM_Cubic, leaveTangent=0.7489, time=0.5, value=1}, <at 1 1.75 RCIM_Cubic>]}
        }}
       Scale Curve=&MakeVector(
        X=&Lerp_Float(A=0, B=0.15, Alpha=$User.uPhase)
        Y=&Lerp_Float(A=0.15, B=0.35, Alpha=$User.uPhase)
        Z=&Lerp_Float(A=0.25, B=0.5, Alpha=$User.uPhase)
       )
      )
      B=&VectorFromFloat(Value=&FloatFromCurve(
        FloatCurve=DataInterface() {propertyValues {Curve {keys [
            {arriveTangent=5.1112, interpMode=RCIM_Cubic, leaveTangent=5.1112, tangentMode=RCTM_User, time=0, value=0}
            {arriveTangent=-4.9135, interpMode=RCIM_Cubic, leaveTangent=-4.9135, tangentMode=RCTM_User, time=1, value=0.2}
           ]}}}
        Scale Curve=&Lerp_Float(
         A=$User.FuelFlicker
         B=1
         Alpha=&Waveform("Waveform [1]"=::NewEnumerator8, "[1] Phase Scale"=50, "[1] Frequency"=5, "[1] Amplitude Min/Max"=<0 1>)
        )
       ))
     )
     Float=1
    )
   ) {isSetParameters=true, module=SetVariables_#2}
   &ParticleState()
   &SkeletalMeshLocation(
    Bone / Socket Index=$Particles.PartitionParticles.Partition
    MaskSpawnGroup=false
    Mesh Sampling Type=::NewEnumerator5
    Sampled Position Offset=<1 0 0>
    Skeletal Mesh=$User.Skeletal Mesh
    Socket Sampling Mode=::NewEnumerator5
   )
   /SetVariables(Particles.MeshOrientation=$Particles.SkeletalMeshLocation.SampledRotation) {isSetParameters=true, module=SetVariables_#3}
  ]
  rendererClasses [&NiagaraMeshRendererProperties]
  renderers [{rendererClass=&NiagaraMeshRendererProperties, rendererIndex=0}]
 }
 Combined_Rate {
  emitterUpdateScript [
   &EmitterState()
   &SpawnRate(SpawnRate=24, Use Spawn Probability=false)
  ]
  particleSpawnScript [
   &PartitionParticles()
   &InitializeParticle(
    Color Mode=::NewEnumerator0
    Lifetime Max=0.35
    Lifetime Min=0.25
    Lifetime Mode=::NewEnumerator1
    Position Mode=::NewEnumerator0
    Write Material Random=true
   )
   /SetVariables(
    Particles.Color_Blaster:LinearColor=<rgba 20 1.6667 0 0.1>
    Particles.Orientation_Blaster:Quat4f=&OrientMeshToVector(Look At Direction=<0 -1 0>, Reference Axis Vector=<1 0 0>)
    Particles.Size_Blaster:Vector3f=<1.25 1 1>
   ) {isSetParameters=true, module=SetVariables_#1}
  ]
  particleUpdateScript [
   /SetVariables(
    Particles.Color_Blaster:LinearColor=&MultiplyLinearColor(
     Linear Color=&MultiplyLinearColor(
      Linear Color=&ColorFromCurve(
       ColorCurve=DataInterface() {propertyValues {
         AlphaCurve {keys [<at 1 1 RCIM_Cubic>]}
         BlueCurve {keys [<at 0 0>,<at 0.15 0>,<at 1 0>]}
         GreenCurve {keys [<at 0 0.125>,<at 0.15 0.6667>,<at 1 0.25>]}
         RedCurve {keys [<at 0 1>,<at 0.15 4>,<at 1 1>]}
        }}
       CurveIndex=$User.FuelFlicker
      )
      Scale Factor=<2 2 2 1>
     )
     Scale Factor=&MakeFloat4FromVectorAndFloat(
      "Vector (RGB)"=&VectorFromCurve(VectorCurve=DataInterface() {propertyValues {
         XCurve {keys [<at 0 2 RCIM_Cubic>,<at 1 1 RCIM_Cubic>]}
         YCurve {keys [<at 0 2 RCIM_Cubic>,<at 1 1 RCIM_Cubic>]}
         ZCurve {keys [<at 0 2 RCIM_Cubic>,<at 1 1 RCIM_Cubic>]}
        }})
      "Float (Alpha)"=&Lerp_Float(
       A=&Waveform("Waveform [1]"=::NewEnumerator8, "[1] Frequency"=1, "[1] Amplitude Min/Max"=<0 1>, "[1] Scale"=2)
       B=&FloatFromCurve(
        FloatCurve=DataInterface() {propertyValues {Curve {keys [<at 0 0 RCIM_Cubic>,<at 0.5 1>,<at 1.0011 0 RCIM_Cubic>]}}}
        Scale Curve=&Lerp_Float(
         A=$User.FuelFlicker
         B=1
         Alpha=&Waveform("Waveform [1]"=::NewEnumerator8, "[1] Phase Scale"=50, "[1] Frequency"=5, "[1] Amplitude Min/Max"=<0 1>)
        )
       )
       Alpha=$User.FuelFlicker
      )
     )
    )
    Particles.DynamicParameter_Blaster:Vector4f=&MakeVector4(
     X=&FloatFromCurve(FloatCurve=DataInterface() {propertyValues {Curve {keys [<at 0 1>,<at 1 0>]}}})
     Y=5
     Z=20
     W=1
    )
    Particles.Size_Blaster:Vector3f=&Multiply_VectorByFloat(
     Vector=&Multiply_Vector(
      A=&VectorFromCurve(
       VectorCurve=DataInterface() {propertyValues {
         XCurve {keys [
           <at 0 0.75 RCIM_Cubic>
           <at 0.4908 3.9579>
           {arriveTangent=4.355, interpMode=RCIM_Cubic, leaveTangent=4.355, tangentMode=RCTM_User, time=0.5, value=4}
           <at 1 0.5 RCIM_Cubic>
          ]}
         YCurve {keys [
           <at 0 1 RCIM_Cubic>
           <at 0.4908 0.9933>
           {arriveTangent=0.75, interpMode=RCIM_Cubic, leaveTangent=0.75, time=0.5, value=1}
           <at 1 1.75 RCIM_Cubic>
          ]}
         ZCurve {keys [<at -0.0014 1 RCIM_Cubic>, {arriveTangent=0.7489, interpMode=RCIM_Cubic, leaveTangent=0.7489, time=0.5, value=1}, <at 1 1.75 RCIM_Cubic>]}
        }}
       Scale Curve=&MakeVector(
        X=&Lerp_Float(A=0, B=0.075, Alpha=$User.uPhase)
        Y=&Lerp_Float(A=0.15, B=0.35, Alpha=$User.uPhase)
        Z=&Lerp_Float(A=0.25, B=0.5, Alpha=$User.uPhase)
       )
      )
      B=&VectorFromFloat(Value=&FloatFromCurve(
        FloatCurve=DataInterface() {propertyValues {Curve {keys [
            {arriveTangent=5.1112, interpMode=RCIM_Cubic, leaveTangent=5.1112, tangentMode=RCTM_User, time=0, value=0}
            {arriveTangent=-4.9135, interpMode=RCIM_Cubic, leaveTangent=-4.9135, tangentMode=RCTM_User, time=1, value=0.2}
           ]}}}
        Scale Curve=&Lerp_Float(
         A=$User.FuelFlicker
         B=1
         Alpha=&Waveform("Waveform [1]"=::NewEnumerator8, "[1] Phase Scale"=50, "[1] Frequency"=5, "[1] Amplitude Min/Max"=<0 1>)
        )
       ))
     )
     Float=1
    )
   ) {isSetParameters=true, module=SetVariables_#2}
   &ParticleState()
   &SkeletalMeshLocation(
    Bone / Socket Index=$Particles.PartitionParticles.Partition
    MaskSpawnGroup=false
    Mesh Sampling Type=::NewEnumerator5
    Sampled Position Offset=<7 0 0>
    Skeletal Mesh=$User.Skeletal Mesh
    Socket Sampling Mode=::NewEnumerator5
   )
   /SetVariables(Particles.MeshOrientation=$Particles.SkeletalMeshLocation.SampledRotation) {isSetParameters=true, module=SetVariables_#3}
   &DynamicMaterialParameters(
    Index 0 Param 2=1
    Module.Index0Param0Write=true
    Module.Index0Param1Write=true
    Module.Index0Param2Write=true
    Module.Index0Param3Write=true
   )
  ]
  rendererClasses [&NiagaraMeshRendererProperties]
  renderers [{rendererClass=&NiagaraMeshRendererProperties, rendererIndex=0}]
 }
 Combined_Rate001 {
  emitterUpdateScript [
   &EmitterState()
   &SpawnRate(SpawnRate=24, Use Spawn Probability=false)
  ]
  particleSpawnScript [
   &PartitionParticles()
   &InitializeParticle(
    Color Mode=::NewEnumerator0
    Lifetime Max=0.35
    Lifetime Min=0.25
    Lifetime Mode=::NewEnumerator1
    Position Mode=::NewEnumerator0
    Write Material Random=true
   )
   /SetVariables(
    Particles.Color_Blaster:LinearColor=<rgba 20 1.6667 0 0.1>
    Particles.Orientation_Blaster:Quat4f=&OrientMeshToVector(Look At Direction=<0 -1 0>, Reference Axis Vector=<1 0 0>)
    Particles.Size_Blaster:Vector3f=<1.25 1 1>
   ) {isSetParameters=true, module=SetVariables_#1}
  ]
  particleUpdateScript [
   /SetVariables(
    Particles.Color_Blaster:LinearColor=&MultiplyLinearColor(
     Linear Color=&MultiplyLinearColor(
      Linear Color=&ColorFromCurve(
       ColorCurve=DataInterface() {propertyValues {
         AlphaCurve {keys [<at 0 0 RCIM_Cubic>,<at 0.15 1>]}
         BlueCurve {keys [<at 1 2 RCIM_Cubic>]}
         GreenCurve {keys [<at 1 0.8803 RCIM_Cubic>]}
         RedCurve {keys [<at 1 0 RCIM_Cubic>]}
        }}
       CurveIndex=$User.FuelFlicker
      )
      Scale Factor=<3 3 3 1>
     )
     Scale Factor=&MakeFloat4FromVectorAndFloat(
      "Vector (RGB)"=&VectorFromCurve(VectorCurve=DataInterface() {propertyValues {
         XCurve {keys [<at 0 2 RCIM_Cubic>,<at 1 1 RCIM_Cubic>]}
         YCurve {keys [<at 0 2 RCIM_Cubic>,<at 1 1 RCIM_Cubic>]}
         ZCurve {keys [<at 0 2 RCIM_Cubic>,<at 1 1 RCIM_Cubic>]}
        }})
      "Float (Alpha)"=&Lerp_Float(
       A=&Waveform("Waveform [1]"=::NewEnumerator8, "[1] Frequency"=1, "[1] Amplitude Min/Max"=<0 1>, "[1] Scale"=2)
       B=&FloatFromCurve(
        FloatCurve=DataInterface() {propertyValues {Curve {keys [<at 0 0 RCIM_Cubic>,<at 0.5 1>,<at 1.0011 0 RCIM_Cubic>]}}}
        Scale Curve=&Lerp_Float(
         A=$User.FuelFlicker
         B=1
         Alpha=&Waveform("Waveform [1]"=::NewEnumerator8, "[1] Phase Scale"=50, "[1] Frequency"=5, "[1] Amplitude Min/Max"=<0 1>)
        )
       )
       Alpha=$User.FuelFlicker
      )
     )
    )
    Particles.DynamicParameter_Blaster:Vector4f=&MakeVector4(
     X=&FloatFromCurve(FloatCurve=DataInterface() {propertyValues {Curve {keys [<at 0 1>,<at 1 0>]}}})
     Y=5
     Z=20
     W=1
    )
    Particles.Size_Blaster:Vector3f=&Multiply_VectorByFloat(
     Vector=&Multiply_Vector(
      A=&VectorFromCurve(
       VectorCurve=DataInterface() {propertyValues {
         XCurve {keys [
           <at 0 0.75 RCIM_Cubic>
           <at 0.4908 3.9579>
           {arriveTangent=4.355, interpMode=RCIM_Cubic, leaveTangent=4.355, tangentMode=RCTM_User, time=0.5, value=4}
           <at 1 0.5 RCIM_Cubic>
          ]}
         YCurve {keys [
           <at 0 1 RCIM_Cubic>
           <at 0.4908 0.9933>
           {arriveTangent=0.75, interpMode=RCIM_Cubic, leaveTangent=0.75, time=0.5, value=1}
           <at 1 1.75 RCIM_Cubic>
          ]}
         ZCurve {keys [<at -0.0014 1 RCIM_Cubic>, {arriveTangent=0.7489, interpMode=RCIM_Cubic, leaveTangent=0.7489, time=0.5, value=1}, <at 1 1.75 RCIM_Cubic>]}
        }}
       Scale Curve=&MakeVector(
        X=&Lerp_Float(A=0, B=0.035, Alpha=$User.uPhase)
        Y=&Lerp_Float(A=0.15, B=0.2, Alpha=$User.uPhase)
        Z=&Lerp_Float(A=0.2, B=0.25, Alpha=$User.uPhase)
       )
      )
      B=&VectorFromFloat(Value=&FloatFromCurve(
        FloatCurve=DataInterface() {propertyValues {Curve {keys [
            {arriveTangent=5.1112, interpMode=RCIM_Cubic, leaveTangent=5.1112, tangentMode=RCTM_User, time=0, value=0}
            {arriveTangent=-4.9135, interpMode=RCIM_Cubic, leaveTangent=-4.9135, tangentMode=RCTM_User, time=1, value=0.2}
           ]}}}
        Scale Curve=&Lerp_Float(
         A=$User.FuelFlicker
         B=1
         Alpha=&Waveform("Waveform [1]"=::NewEnumerator8, "[1] Phase Scale"=50, "[1] Frequency"=5, "[1] Amplitude Min/Max"=<0 1>)
        )
       ))
     )
     Float=1
    )
   ) {isSetParameters=true, module=SetVariables_#2}
   &ParticleState()
   &SkeletalMeshLocation(
    Bone / Socket Index=$Particles.PartitionParticles.Partition
    MaskSpawnGroup=false
    Mesh Sampling Type=::NewEnumerator5
    Sampled Position Offset=<7 0 0>
    Skeletal Mesh=$User.Skeletal Mesh
    Socket Sampling Mode=::NewEnumerator5
   )
   /SetVariables(Particles.MeshOrientation=$Particles.SkeletalMeshLocation.SampledRotation) {isSetParameters=true, module=SetVariables_#3}
  ]
  rendererClasses [&NiagaraMeshRendererProperties]
  renderers [{rendererClass=&NiagaraMeshRendererProperties, rendererIndex=0}]
 }
 Head {
  bEnabled=false
  emitterUpdateScript [
   &EmitterState()
   &SpawnRate(SpawnRate=500, Use Spawn Probability=true)
  ]
  particleSpawnScript [
   &PartitionParticles()
   &InitializeParticle(
    Color Mode=::NewEnumerator0
    Lifetime Max=0.1
    Lifetime Min=0.1
    Lifetime Mode=::NewEnumerator1
    Ribbon Width Mode=::NewEnumerator1
    Sprite Size Mode=::NewEnumerator1
    UsePositionOffset=false
    Write Material Random=true
   )
   &SkeletalMeshLocation(
    Bone / Socket Index=$Particles.PartitionParticles.Partition
    MaskSpawnGroup=false
    Mesh Sampling Type=::NewEnumerator5
    Skeletal Mesh=$User.Skeletal Mesh
    Socket Sampling Mode=::NewEnumerator5
   )
   &AddVelocity(
    Additional Quaternion Rotation=false
    Rotation Mode=::NewEnumerator3
    Rotation Quaternion:Quat4f=$Particles.MeshOrientation
    Velocity=<1500 0 -50>
   )
  ]
  particleUpdateScript [
   &ParticleState()
   &ScaleColor(
    Scale Mode=::NewEnumerator1
    Scale RGBA=&MakeVector4_FromLinearColor(LinearColor=&LerpLinearColors(
      StartColor=&ColorFromCurve(ColorCurve=DataInterface() {propertyValues {
         AlphaCurve {keys [<at 0 1 RCIM_Cubic>,<at 1 0 RCIM_Cubic>]}
         BlueCurve {keys [<at 0 0 RCIM_Cubic>,<at 0.2335 1 RCIM_Cubic>]}
         GreenCurve {keys [<at 0 1.629 RCIM_Cubic>,<at 0.2335 1 RCIM_Cubic>]}
         RedCurve {keys [<at 0 50 RCIM_Cubic>,<at 0.2335 1 RCIM_Cubic>]}
        }})
      EndColor=&ColorFromCurve(ColorCurve=DataInterface() {propertyValues {
         AlphaCurve {keys [<at 0 1 RCIM_Cubic>,<at 1 0 RCIM_Cubic>]}
         BlueCurve {keys [<at 0 0 RCIM_Cubic>,<at 0.3842 1 RCIM_Cubic>]}
         GreenCurve {keys [<at 0 0.2003 RCIM_Cubic>,<at 0.3842 1 RCIM_Cubic>]}
         RedCurve {keys [<at 0 1 RCIM_Cubic>,<at 0.3842 1 RCIM_Cubic>]}
        }})
      LerpFactor=$User.FuelFlicker
     ))
    ScaleRGBA=true
   )
   &Drag(Use Linear Drag=true, Use Rotational Drag=true)
   &SolveForcesAndVelocity(Clamp Velocity=false, Limit Acceleration=false)
   &ScaleRibbonWidth(Ribbon Width Scale=&FloatFromCurve(
     FloatCurve=DataInterface() {propertyValues {Curve {keys [{interpMode=RCIM_Cubic, tangentMode=RCTM_Break, time=0, value=1}, <at 1 0>]}}}
     Scale Curve=2
    ))
   &DynamicMaterialParameters(
    Index 0 Param 1=&FloatFromCurve(FloatCurve=DataInterface() {propertyValues {Curve {keys [{interpMode=RCIM_Cubic, tangentMode=RCTM_Break, time=0, value=1}, {interpMode=RCIM_Cubic, tangentMode=RCTM_Break, time=1, value=0}]}}})
    Index 0 Param 3=&SelectFloatFromArray(Direct Array Index=$Particles.PartitionParticles.Partition, Float Selection Array=DataInterface() {propertyValues {FloatData [-1.5,-1]}})
    Index 0 Param 4=0.25
    Module.Index0Param0Write=true
    Module.Index0Param1Write=true
    Module.Index0Param2Write=true
    Module.Index0Param3Write=true
   )
   &ScaleColor(
    Scale Alpha=&Lerp_Float(
     A=&Waveform("Waveform [1]"=::NewEnumerator8, "[1] Frequency"=1, "[1] Amplitude Min/Max"=<0.125 1>)
     B=1
     Alpha=$User.FuelFlicker
    )
    ScaleA=true
    ScaleRGB=true
   ) {module=ScaleColor001}
  ]
  rendererClasses [&NiagaraSpriteRendererProperties, &NiagaraRibbonRendererProperties]
  renderers [
   {rendererClass=&NiagaraSpriteRendererProperties}
   {rendererClass=&NiagaraRibbonRendererProperties, rendererIndex=1}
  ]
 }
 Smoke {
  emitterUpdateScript [
   &EmitterState()
   &SpawnBurst_Instantaneous(
    Spawn Count=&RandomRangeInt(Minimum=3, Maximum=6)
    Use Spawn Probability=false
    UseLoopCountLimit=false
   )
   &SpawnRate(
    SpawnRate=&Multiply_Float(A=3, B=&Lerp_Float(A=0, B=1, Alpha=$Emitter.NormalizedLoopAge))
    Use Spawn Probability=false
   )
  ]
  particleSpawnScript [
   &PartitionParticles()
   &InitializeParticle(
    Color Mode=::NewEnumerator0
    Lifetime Max=0.5
    Lifetime Min=0.25
    Lifetime Mode=::NewEnumerator1
    Sprite Size Mode=::NewEnumerator1
    Uniform Sprite Size Max=60
    Uniform Sprite Size Min=30
    UsePositionOffset=false
    Write Material Random=true
   )
   &SkeletalMeshLocation(
    Bone / Socket Index=$Particles.PartitionParticles.Partition
    Interpolate Mesh Sampling=false
    MaskSpawnGroup=false
    Mesh Sampling Type=::NewEnumerator5
    Sampled Position Offset=<7 0 0>
    Skeletal Mesh=$User.Skeletal Mesh
    Socket Sampling Mode=::NewEnumerator5
   )
   &AddVelocity(
    Additional Quaternion Rotation=false
    Rotation Mode=::NewEnumerator3
    Rotation Quaternion:Quat4f=$Particles.SkeletalMeshLocation.SampledRotation
    Velocity=<1 0 0>
    Velocity Speed Scale=&RandomRangeFloat(Minimum=0, Maximum=250)
   )
   &AddVelocity(
    Cone Axis=&RotateVectorByQuaternion(Vector To Rotate=<1 0 0>, Quaternion=$Particles.SkeletalMeshLocation.SampledRotation)
    Inner Cone Angle=180
    Module.UseConeFalloff=false
    Module.UseVelDistribution=false
    Override Seed=false
    Velocity Mode=::NewEnumerator2
    Velocity Speed=&RandomRangeFloat(Minimum=0, Maximum=150)
   ) {module=AddVelocity001}
   &ShapeLocation(MaskSpawnGroup=false, Override Seed=false, Sphere Radius=5, UseOwnerScale=true)
   &InheritVelocity(
    Inherited Velocity Amount Scale=&Multiply_VectorByFloat(Vector=<-0.99 -0.99 -0.99>, Float=&RandomRangeFloat(Minimum=0.99, Maximum=1.1))
    Speed Limit=false
    UseSpeedThreshold=false
   ) {enabled=false}
  ]
  particleUpdateScript [
   &Drag(Drag=&RandomRangeFloat(Minimum=4, Maximum=8), Use Linear Drag=true, Use Rotational Drag=true)
   &AccelerationForce() {enabled=false}
   &SolveForcesAndVelocity(Clamp Velocity=false, Limit Acceleration=false)
   &ParticleState()
   &ScaleSpriteSize(Uniform Curve Sprite Scale=DataInterface() {propertyValues {Curve {keys [
        {arriveTangent=1.7491, interpMode=RCIM_Cubic, leaveTangent=1.7491, tangentMode=RCTM_User, time=0, value=0}
        {tangentMode=RCTM_User, time=1, value=1}
       ]}}})
   &ScaleColor(
    Linear Color Curve=DataInterface() {propertyValues {AlphaCurve {keys [<at 0 1>,<at 1 0>]}, BlueCurve {keys [<at 0 0>]}, GreenCurve {keys [<at 0 0.0041>]}, RedCurve {keys [<at 0 0.0573>]}}}
    Scale Mode=::NewEnumerator2
   )
  ]
  rendererClasses [&NiagaraSpriteRendererProperties]
  renderers [{rendererClass=&NiagaraSpriteRendererProperties}]
 }
 Sparks {
  emitterUpdateScript [
   &EmitterState()
   &SpawnBurst_Instantaneous(
    Spawn Count=&RandomRangeInt(Minimum=7, Maximum=30)
    Use Spawn Probability=false
    UseLoopCountLimit=false
   )
   &SpawnRate(
    SpawnRate=&Multiply_Float(A=64, B=&Lerp_Float(A=0, B=1, Alpha=$Emitter.NormalizedLoopAge))
    Use Spawn Probability=false
   )
  ]
  particleSpawnScript [
   &PartitionParticles()
   &InitializeParticle(
    Color Mode=::NewEnumerator0
    Lifetime Max=0.25
    Lifetime Min=0.1
    Lifetime Mode=::NewEnumerator1
    Sprite Size Mode=::NewEnumerator1
    Uniform Sprite Size Max=3
    Uniform Sprite Size Min=2
    UsePositionOffset=false
    Write Material Random=true
   )
   &SkeletalMeshLocation(
    Bone / Socket Index=$Particles.PartitionParticles.Partition
    Interpolate Mesh Sampling=false
    MaskSpawnGroup=false
    Mesh Sampling Type=::NewEnumerator5
    Sampled Position Offset=<7 0 0>
    Skeletal Mesh=$User.Skeletal Mesh
    Socket Sampling Mode=::NewEnumerator5
   )
   &AddVelocity(
    Additional Quaternion Rotation=false
    Rotation Mode=::NewEnumerator3
    Rotation Quaternion:Quat4f=$Particles.SkeletalMeshLocation.SampledRotation
    Velocity=<1 0 0>
    Velocity Speed Scale=&RandomRangeFloat(Minimum=100, Maximum=500)
   )
   &AddVelocity(
    Cone Axis=&RotateVectorByQuaternion(Vector To Rotate=<1 0 0>, Quaternion=$Particles.SkeletalMeshLocation.SampledRotation)
    Inner Cone Angle=90
    Module.UseConeFalloff=false
    Module.UseVelDistribution=false
    Override Seed=false
    Velocity Mode=::NewEnumerator2
    Velocity Speed=&RandomRangeFloat(Minimum=25, Maximum=150)
   ) {module=AddVelocity001}
   &ShapeLocation(MaskSpawnGroup=false, Override Seed=false, Sphere Radius=5, UseOwnerScale=true)
   &InheritVelocity(
    Inherited Velocity Amount Scale=&Multiply_VectorByFloat(Vector=<-0.95 -0.95 -0.95>, Float=&RandomRangeFloat(Minimum=0.5, Maximum=1))
    Speed Limit=false
    UseSpeedThreshold=false
   ) {enabled=false}
  ]
  particleUpdateScript [
   &SolveForcesAndVelocity(Clamp Velocity=false, Limit Acceleration=false)
   &ParticleState()
   &ScaleSpriteSize(Uniform Curve Sprite Scale=DataInterface() {propertyValues {Curve {keys [
        {arriveTangent=2.4854, interpMode=RCIM_Cubic, leaveTangent=2.4854, tangentMode=RCTM_User, time=0, value=0}
        {tangentMode=RCTM_User, time=1, value=1}
       ]}}})
   &ScaleSpriteSizeBySpeed(Max Scale Factor=<0.5 2>, Min Scale Factor=<0.5 1>)
   &ScaleColor(
    Linear Color Curve=DataInterface() {propertyValues {
      AlphaCurve {keys [<at 0 1>,<at 1 0>]}
      BlueCurve {keys [<at 0 0>,<at 1 0>]}
      GreenCurve {keys [<at 0 1.2515>,<at 1 0.0707>]}
      RedCurve {keys [<at 0 5>,<at 1 1>]}
     }}
    Scale Mode=::NewEnumerator2
   )
  ]
  rendererClasses [&NiagaraSpriteRendererProperties]
  renderers [{rendererClass=&NiagaraSpriteRendererProperties}]
 }
}
summary {
 systemName=NS_JethroJelly_HoverJets_AscendFail
 emitters [
  {emitterName=Combined_Rate, rendererClasses [&NiagaraMeshRendererProperties]}
  {
   emitterName=Airglow_Light_Flare
   rendererClasses [&NiagaraSpriteRendererProperties, &NiagaraLightRendererProperties]
  }
  {emitterName=Sparks, rendererClasses [&NiagaraSpriteRendererProperties]}
  {
   bEnabled=false
   emitterName=Head
   rendererClasses [&NiagaraSpriteRendererProperties, &NiagaraRibbonRendererProperties]
  }
  {emitterName=Burst, rendererClasses [&NiagaraMeshRendererProperties]}
  {emitterName=Combined_Rate001, rendererClasses [&NiagaraMeshRendererProperties]}
  {emitterName=Smoke, rendererClasses [&NiagaraSpriteRendererProperties]}
 ]
 userVariables [
  User.FuelFlicker {type=NiagaraFloat, defaultValue=0, description=""}
  User.Skeletal Mesh {
   type=NiagaraDataInterfaceSkeletalMesh
   defaultValue {
    struct=NiagaraExt_VariableValue_DataInterface
    value {dataInterface=&self:NiagaraDataInterfaceSkeletalMesh_0}
   }
  }
  User.uPhase {type=NiagaraFloat, defaultValue=1, description=""}
 ]
}
